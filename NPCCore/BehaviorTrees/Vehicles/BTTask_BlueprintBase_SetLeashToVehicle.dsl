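; Behavior tree task for NPC bots that moves the AI's leash to a vehicle.
; It reads the vehicle actor from the blackboard and casts the owner to a FortAthenaAIBotController.
; With a valid vehicle, it removes the patrol path from any patrolling component on the controller, then sets the radial leash location to the vehicle's current location.
; The task only ever finishes with success and has no failure result.
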(event EventReceiveExecute
  (params (OwnerActor :type Actor%20Object%20Reference))
  (bind _vehicleactor (Variables|Default|GetVehicleActor))
  (bind _returnvalue (AI|BehaviorTree|GetBlackboardValueasActor _vehicleactor))
  (bind _asfort_athena_aibot_controller (Utilities|Casting|CastToFortAthenaAIBotController OwnerActor)
    (:then
      (Utilities|IsValid _returnvalue
        (:Is%20Valid
          (Utilities|IsValid (Actor|GetComponentbyClass _asfort_athena_aibot_controller "/Script/FortniteAI.FortAthenaNpcPatrollingComponent")
            (:Is%20Valid
              (Class|FortAthenaNpcPatrollingComponent|RemovePatrolPath (Actor|GetComponentbyClass _asfort_athena_aibot_controller "/Script/FortniteAI.FortAthenaNpcPatrollingComponent")))
            (:Is%20Not%20Valid
              _))
          (Class|FortAthenaAIBotController|SetRadialLeashLocation _asfort_athena_aibot_controller (Transformation|GetActorLocation _returnvalue))
          (AI|BehaviorTree|FinishExecute true))
        (:Is%20Not%20Valid
          _)))
    (:CastFailed
      _))
  (AI|BehaviorTree|FinishExecute true))
