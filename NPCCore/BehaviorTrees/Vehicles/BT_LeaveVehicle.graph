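# Behavior tree that makes an NPC leave its vehicle, using the BB_InVehicle blackboard.
# A root Sequence runs a short randomized wait, an optional leash update, then the LeaveVehicle task without waiting for the vehicle to stop.
# The leash step sets the leash to the blackboard's vehicle actor when its blackboard flag allows it.
# An Always Succeed fallback in that Selector keeps the exit from being blocked.
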BehaviorTree /NPCCore/BehaviorTrees/Vehicles/BT_LeaveVehicle
blackboard=/AICore/Humanoids/BehaviorTrees/Blackboards/BB_InVehicle
root {
 name=Sequence
 class=BTComposite_Sequence
 children [
  Wait {class=BTTask_Wait, properties {randomDeviation {defaultValue=0.5}, waitTime {defaultValue=1}}}
  Selector {
   class=BTComposite_Selector
   children [
    Selector {
     class=BTComposite_Selector
     children [BTTask_BlueprintBase_SetLeashToVehicle_C_0 {
       class=BTTask_BlueprintBase_SetLeashToVehicle_C
       children []
       decorators []
       properties {
        bIgnoreRestartSelf=false
        bShowPropertyDetails=true
        customDescription=""
        nodeName=""
        tickInterval {interval=-1}
        vehicleActor {bNoneIsAllowedValue=false, selectedKeyName=AIEvaluator_Vehicle_Actor}
       }
       services []
      }]
     decorators [Should Update Leash When Leaving {
       class=BTDecorator_Blackboard
       properties {
        arithmeticOperation=Equal
        flowAbortMode=None
        intValue=0
        nodeName=Should Update Leash When Leaving
        stringValue=ExecutionError
        blackboardKey {selectedKeyName=AIEvaluator_SetLeashAroundVehicleOnExitVehicle}
       }
      }]
     properties {
      children [{childComposite=null, childTask=&self:BTTask_BlueprintBase_SetLeashToVehicle_C_0, decorators []}]
      services []
     }
     services []
    }
    Always Succeed {class=FortBTTask_AlwaysSucceed, children [], decorators [], properties {bIgnoreRestartSelf=false, nodeName=Always Succeed}, services []}
   ]
   decorators []
   properties {
    children [
     {
      childComposite=&self:BTComposite_Selector_0
      childTask=null
      decorators [&self:BTDecorator_Blackboard_1]
     }
     {childComposite=null, childTask=&self:FortBTTask_AlwaysSucceed_0, decorators []}
    ]
    services []
   }
   services []
  }
  LeaveVehicle {
   class=FortAthenaBTTask_LeaveVehicle
   children []
   decorators []
   properties {bIgnoreRestartSelf=false, bWaitVehicleStop=false, nodeName=LeaveVehicle}
   services []
  }
 ]
 decorators []
 properties {children [
   {childTask=&self:BTTask_Wait_0}
   {childComposite=&self:BTComposite_Selector_1, childTask=null}
   {childTask=&self:FortAthenaBTTask_LeaveVehicle_1}
  ]}
}
